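; Bouncy inflatable Blueprint for the Clyde game, providing wobble, sound and helper checks for bouncing an overlapping pawn.
; MakeWobbleEffect starts a timed bounce animation at a given center, and DoOnce-gated events play the tire sound to avoid double triggers.
; CeilingCheck capsule-traces above the root, ignoring the pawn and two levels of attached actors.
; CheckJumpingOrientation returns a scaled horizontal forward vector, flipping the multiplier's sign when not X-aligned.
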(event Custom|MakeWobbleEffect
  (params (Center :type Vector))
  (CallFunction|StartBounceAnimation (Utilities|Time|GetTimeSeconds) 2e+02 (Math|Vector|MakeVector 0.0 0.0 1.0) Center))

(event Custom|Reset%20Bounce%20Sound
  ;[1 1 1] Keeps bounce sound from double triggering (hopefully)
  (Utilities|FlowControl|DoOnce false
    (:Completed
      (Audio|PlaySoundatLocation (Variables|Default|GetTireSound) (Transformation|GetActorLocation (Variables|Default|GetOverlapPawn))))))

(event Custom|CustomEvent
  ;[1 1 1] Keeps bounce sound from double triggering (hopefully)
  (Utilities|FlowControl|DoOnce false
    (:Completed
      (Audio|PlaySoundatLocation (Variables|Default|GetTireSound) (Transformation|GetActorLocation 0)))))

(fn CeilingCheck
  (bind _overlappawn (Variables|Default|GetOverlapPawn)
    (:then
      (Utilities|Array|Add (LocalVariables|GetActorsToIgnore) _overlappawn)
      (Utilities|Array|AppendArray (LocalVariables|GetActorsToIgnore) (Actor|GetAttachedActors _overlappawn true false))
      (for _array_element (Actor|GetAttachedActors _overlappawn true false)
        (Utilities|Array|AppendArray (LocalVariables|GetActorsToIgnore) (Actor|GetAttachedActors _array_element true false))))
    (:else
      _))
  (bind (_outhit _returnvalue) (Collision|CapsuleTraceByChannel (Math|Vector|MakeVector (.x (Transformation|GetWorldLocation (Variables|Transformation|GetRootComponent))) (.y (Transformation|GetWorldLocation (Variables|Transformation|GetRootComponent))) (+ (.z (Transformation|GetWorldLocation (Variables|Transformation|GetRootComponent))) 200)) (Math|Vector|MakeVector (.x (Transformation|GetWorldLocation (Variables|Transformation|GetRootComponent))) (.y (Transformation|GetWorldLocation (Variables|Transformation|GetRootComponent))) (+ (.z (Transformation|GetWorldLocation (Variables|Transformation|GetRootComponent))) 340)) 50 100 "TraceTypeQuery9" false (LocalVariables|GetActorsToIgnore)))
  (return _returnvalue))

(fn CheckJumpingOrientation
  (bind (_returnvalue_x _returnvalue_y) (Math|Vector|BreakVector (Transformation|GetActorForwardVector)))
  (bind _orientationmultiplier (Variables|Default|GetOrientationMultiplier))
  (if (== (Math|Float|Round (Math|Float|Absolute(Float) _returnvalue_x)) 1)
    (return (Math|Vector|vector*float (Math|Vector|MakeVector _returnvalue_x _returnvalue_y) _orientationmultiplier))
    (else
      (return (Math|Vector|vector*float (Math|Vector|MakeVector _returnvalue_x _returnvalue_y) (* _orientationmultiplier -1.0))))))

(fn ConstructionScript)
